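# Material function for the GateHound pickaxe that applies MF_Crystal through a channel-selectable TechArt mask and adds a wispy smoke effect.
# Smoke comes from panning grayscale and noise textures in pivot-relative preskinned XZ space, alongside a panning Styx normal for distortion.
# Preskinned-position gradient and limit masks help bound the smoke.
# Smoke-driven lerps scale the crystal's BaseColor and Emissive, which are written back into the material attributes.
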MaterialFunction /BRCosmetics/Weapons/FORT_Melee/Pickaxe_GateHound/Materials/MF_GateHoundCrystal
graph {nodes [
  n1 = FunctionInput() {bUsePreviewValueAsDefault=true, inputName=TechArtMask, sortPriority=1, previewValue=<1 0 0 1>}
  n2 = ChannelMaskParameter(None=@n1) {group=Crystal, maskChannel=Green, parameterName=UseCrystal, defaultValue=<rgba 0 1 0 0>}
  n3 = GetMaterialAttributes(None=MaterialFunctionCall!(
    Material=FunctionInput() {inputName=Material, inputType=FunctionInput_MaterialAttributes}
    Alpha=NamedRerouteDeclaration(None=@n2.R) {name=TechArtMask, nodeColor=<rgba 0 1 0.3203 1>}
   ) {materialFunction=/Game/Global/MaterialFunctions/TechArt/MF_Crystal}.Result) {attributeGetTypes [#1, #2]}
  n4 = NamedRerouteUsage()
  n5 = NamedRerouteUsage()
  n6 = Multiply(A=@n5) {constB=10}
  n7 = Multiply(
   A=@n4
   B=LinearInterpolate(B=ScalarParameter() {defaultValue=0.2, group=CrystalSmoke, parameterName=EmissiveColorSmoke}, Alpha=Saturate(None=@n6)) {constA=1, constB=0.2}
  )
  n8 = NamedRerouteUsage()
  n9 = NamedRerouteUsage()
  n10 = Multiply(A=@n5) {constB=7}
  n11 = Multiply(
   A=@n9
   B=LinearInterpolate(B=ScalarParameter() {defaultValue=0.2, group=CrystalSmoke, parameterName=BaseColorSmoke}, Alpha=Saturate(None=@n10)) {constA=1, constB=0.2}
  )
  n12 = NamedRerouteUsage()
  FunctionOutput(None=SetMaterialAttributes!(
    MaterialAttributes=@n3.MaterialAttributes
    EmissiveColor=LinearInterpolate(A=@n4, B=@n7, Alpha=@n8)
    BaseColor=LinearInterpolate(A=@n9, B=@n11, Alpha=@n12)
   ) {attributeSetTypes [#1, #2]})
  n13 = ScalarParameter() {group=Styx, parameterName=StyxNoise_SpeedX}
  n14 = ScalarParameter() {defaultValue=0.05, group=Styx, parameterName=StyxNoise_SpeedY}
  n15 = TextureObjectParameter() {
   group=Styx
   parameterName=StyxDistortNormal
   samplerType=SAMPLERTYPE_Normal
   texture=/BRCosmetics/Characters/Player/Male/Medium/Bodies/M_MED_GateHound_Spectral/Textures/T_StyxGateHound_N
  }
  n16 = TextureSample(UVs=Panner(Coordinate=NamedRerouteUsage(), Speed=AppendVector(A=@n13, B=@n14)), Tex=@n15) {
   samplerSource=SSM_Wrap_WorldGroupSettings
   samplerType=SAMPLERTYPE_Normal
   texture=/BRCosmetics/Characters/Player/Male/Medium/Bodies/M_MED_DarkStance_Pater/Textures/T_StyxCosmetic_N
  }
  NamedRerouteDeclaration(None=Multiply(A=ComponentMask(None=@n16.RGB) {g=true, r=true}, B=ScalarParameter() {defaultValue=0.1, group=Styx, parameterName=DistortIntensity})) {name=StyxDistort, nodeColor=<rgba 1 0 0.9609 1>}
  n17 = FunctionInput() {bUsePreviewValueAsDefault=true, inputName=PreskinnedLoc, sortPriority=2, previewValue=<1 0 0 1>}
  n18 = NamedRerouteDeclaration(None=Subtract(A=@n17, B=VectorParameter() {group=CrystalSmoke, parameterName=SmokePivot, defaultValue=<rgba 0 -8 -15 1>}.RGB)) {name=PreskinnedPos, nodeColor=<rgba 0 0.3437 1 1>}
  n19 = Abs(None=Divide(A=ComponentMask(None=@n18) {g=true}, B=ScalarParameter() {defaultValue=5, group=CrystalSmoke, parameterName=PreskinnedUnitG}))
  NamedRerouteDeclaration(None=SmoothStep(Value=Saturate(None=@n19)) {constMax=0.8, constMin=0.2}) {name=GradientMask, nodeColor=<rgba 0 1 0.3203 1>}
  n20 = ComponentMask(None=@n18) {b=true, r=true}
  n21 = ScalarParameter() {defaultValue=7, group=CrystalSmoke, parameterName=PreskinnedUnit}
  n22 = NamedRerouteDeclaration(None=Divide(A=@n20, B=@n21)) {name=PreskinnedCoord, nodeColor=<rgba 0 0.7891 1 1>}
  n23 = NamedRerouteUsage()
  n24 = ScalarParameter() {group=CrystalSmoke, parameterName=SmokeSpeedX}
  n25 = ScalarParameter() {defaultValue=0.2, group=CrystalSmoke, parameterName=SmokeSpeedY}
  n26 = TextureSampleParameter2D(UVs=Panner(Coordinate=Add(A=@n22, B=@n23), Speed=AppendVector(A=@n24, B=@n25)) {speedY=0.1}) {
   group=CrystalSmoke
   parameterName=SmokeGrayscale
   samplerType=SAMPLERTYPE_Grayscale
   texture=/BRCosmetics/Weapons/FORT_Melee/Pickaxe_GateHound/Textures/T_Chainmail_WispyMaskFat_128
  }
  n27 = Frac(None=Divide(A=@n20, B=Multiply(A=@n21) {constB=6}))
  n28 = Multiply(A=SmoothStep(Value=@n26.R) {constMin=0.4}, B=Subtract(A=@n26.R, B=ComponentMask(None=@n27) {g=true}))
  n29 = TextureSampleParameter2D(UVs=Panner(Coordinate=NamedRerouteUsage()) {speedY=0.1}) {
   group=CrystalSmoke
   parameterName=NoiseSmoke
   samplerType=SAMPLERTYPE_Masks
   texture=/BRCosmetics/Characters/Player/Male/Medium/Bodies/M_MED_GateHound_Spectral/Textures/T_StyxGateHound_DetailNoise_M
  }
  n30 = Multiply(A=Saturate(None=Add(A=@n26.R, B=@n28)), B=@n29.G)
  n31 = NamedRerouteUsage()
  n32 = Saturate(None=NamedRerouteUsage())
  NamedRerouteDeclaration(None=Multiply(A=Multiply(A=@n30, B=@n31), B=LinearInterpolate(Alpha=@n32) {constA=1, constB=0})) {name=SmokeGradient, nodeColor=<rgba 0.5938 0 1 1>}
  NamedRerouteDeclaration(None=@n3.EmissiveColor) {name=EM_IN, nodeColor=<rgba 0.1953 1 0 1>}
  NamedRerouteDeclaration(None=@n3.BaseColor) {name=BaseColor_IN, nodeColor=<rgba 0.2422 0 1 1>}
  n33 = ComponentMask(None=Subtract(A=@n17, B=VectorParameter() {group=CrystalSmoke, parameterName=LimitPivot, defaultValue=<rgba 0 -8 15 1>}.RGB)) {b=true}
  n34 = ScalarParameter() {defaultValue=10, group=CrystalSmoke, parameterName=LimitUnit}
  NamedRerouteDeclaration(None=Saturate(None=Divide(A=@n33, B=@n34) {constB=25})) {name=LimitMask, nodeColor=<rgba 1 0 0.7266 1>}
 ]}
